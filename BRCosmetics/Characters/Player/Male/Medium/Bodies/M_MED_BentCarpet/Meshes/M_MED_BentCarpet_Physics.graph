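# Physics asset for the M_MED_BentCarpet male player body, set up for secondary motion of its outer jacket, undershirt and zipper.
# Capsules on pelvis, spine, thighs and feet form the body volumes, while tiny base capsules anchor jacket chains to spine_02 and undershirt chains to pelvis.
# Dyn bones hang from those anchors with tight, mostly single-axis swing limits, and the zipper hangs off the left jacket.
# Body and anchor joints are mostly unconstrained in rotation.
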PhysicsAsset /BRCosmetics/Characters/Player/Male/Medium/Bodies/M_MED_BentCarpet/Meshes/M_MED_BentCarpet_Physics
bodies {
 base_outer_jacket_l_01 [{shapeType=Capsule, length=0.505, radius=0.505, shapeName=base_outer_jacket_l_01_capsule, rotation=<rot 0 -90 0>}]
 base_outer_jacket_l_03 [{shapeType=Capsule, length=0.505, radius=0.505, shapeName=base_outer_jacket_l_03_capsule, rotation=<rot 0 -90 0>}]
 base_outer_jacket_r_01 [{shapeType=Capsule, length=0.505, radius=0.505, shapeName=base_outer_jacket_r_01_capsule, rotation=<rot 0 -90 0>}]
 base_outer_jacket_r_03 [{shapeType=Capsule, length=0.505, radius=0.505, shapeName=base_outer_jacket_r_03_capsule, rotation=<rot 0 -90 0>}]
 base_undershirt_01 [{shapeType=Capsule, length=0.505, radius=0.505, shapeName=base_undershirt_01_capsule, rotation=<rot 0 -90 0>}]
 base_undershirt_back_01 [{shapeType=Capsule, length=0.505, radius=0.505, shapeName=base_undershirt_back_01_capsule, rotation=<rot 0 -90 0>}]
 base_undershirt_back_l_01 [{shapeType=Capsule, length=0.505, radius=0.505, shapeName=base_undershirt_back_l_01_capsule, rotation=<rot 0 -90 0>}]
 base_undershirt_back_r_01 [{shapeType=Capsule, length=0.505, radius=0.505, shapeName=base_undershirt_back_r_01_capsule, rotation=<rot 0 -90 0>}]
 base_undershirt_l_01 [{shapeType=Capsule, length=0.505, radius=0.505, shapeName=base_undershirt_l_01_capsule, rotation=<rot 0 -90 0>}]
 base_undershirt_r_01 [{shapeType=Capsule, length=0.505, radius=0.505, shapeName=base_undershirt_r_01_capsule, rotation=<rot 0 -90 0>}]
 dyn_minor_zip_jacket [{shapeType=Capsule, length=2, radius=2, shapeName=dyn_minor_zip_jacket_capsule, center=<2.0357 0 0>, rotation=<rot 90 0 0>}]
 dyn_outer_jacket_l_01 [
  {shapeType=Capsule, length=15, radius=3.5, shapeName=dyn_outer_jacket_l_01_capsule, center=<-9.7 0 0>, rotation=<rot 85 80 85>}
  {shapeType=Capsule, length=15, radius=3.5, shapeName=dyn_outer_jacket_l_01_capsule, center=<-9.2167 5.5451 -0.4851>, rotation=<rot 85 80 85>}
 ]
 dyn_outer_jacket_l_03 [{shapeType=Capsule, length=4, radius=5, shapeName=dyn_outer_jacket_l_01_capsule, center=<-6.732 0 0>, rotation=<rot 90 0 0>}]
 dyn_outer_jacket_l_04 [{shapeType=Capsule, length=1, radius=4, shapeName=dyn_outer_jacket_l_04_capsule, center=<-4.5161 0 0>, rotation=<rot 90 0 0>}]
 dyn_outer_jacket_r_01 [{shapeType=Capsule, length=15, radius=4, shapeName=dyn_outer_jacket_l_01_capsule, center=<9.7 0 0>, rotation=<rot 85 80.0033 85.0039>}]
 dyn_outer_jacket_r_03 [{shapeType=Capsule, length=4, radius=5, shapeName=dyn_outer_jacket_l_01_capsule, center=<7 0 0>, rotation=<rot 90 0 0>}]
 dyn_outer_jacket_r_04 [{shapeType=Capsule, length=1, radius=4, shapeName=dyn_outer_jacket_r_04_capsule, center=<4.8322 0 0>, rotation=<rot 90 0 0>}]
 dyn_undershirt_01 [{shapeType=Capsule, length=10, radius=2, shapeName=dyn_undershirt_01_capsule, center=<2.8515 0 0>, rotation=<rot 0 0 90>}]
 dyn_undershirt_02 [{shapeType=Capsule, length=13, radius=3, shapeName=dyn_undershirt_01_capsule, center=<2.8515 -2.8016 0>, rotation=<rot 0 0 90>}]
 dyn_undershirt_back_01 [{shapeType=Capsule, length=1, radius=2, shapeName=dyn_undershirt_back_01_capsule, center=<2.433 0 0>, rotation=<rot 90 0 0>}]
 dyn_undershirt_back_02 [{shapeType=Capsule, length=1, radius=4, shapeName=dyn_undershirt_back_01_capsule, center=<2.433 0 -0.8015>, rotation=<rot 90 0 0>}]
 dyn_undershirt_back_03 [{shapeType=Capsule, length=1, radius=3, shapeName=dyn_undershirt_back_01_capsule, center=<2.433 0 0>, rotation=<rot 90 0 0>}]
 dyn_undershirt_back_l_01 [{shapeType=Capsule, length=1, radius=2, shapeName=dyn_undershirt_back_01_capsule, center=<2.3225 0 0.88>, rotation=<rot 90 0.0002 0>}]
 dyn_undershirt_back_l_02 [{shapeType=Capsule, length=1, radius=3, shapeName=dyn_undershirt_back_01_capsule, center=<2.3225 0 1.6689>, rotation=<rot 90 0.0002 0>}]
 dyn_undershirt_back_l_03 [{shapeType=Capsule, length=1, radius=3, shapeName=dyn_undershirt_back_01_capsule, center=<2.7711 0 1>, rotation=<rot 90 0.0002 0>}]
 dyn_undershirt_back_r_01 [{shapeType=Capsule, length=1, radius=2.5, shapeName=dyn_undershirt_back_01_capsule, center=<-2.2589 0 -0.88>, rotation=<rot 90 0 0>}]
 dyn_undershirt_back_r_02 [{shapeType=Capsule, length=1, radius=3, shapeName=dyn_undershirt_back_01_capsule, center=<-2.2589 0 -1.6>, rotation=<rot 90 0 0>}]
 dyn_undershirt_back_r_03 [{shapeType=Capsule, length=1, radius=3, shapeName=dyn_undershirt_back_01_capsule, center=<-2.2589 0 -1>, rotation=<rot 90 0 0>}]
 dyn_undershirt_l_01 [{shapeType=Capsule, length=3, radius=3, shapeName=dyn_undershirt_01_capsule, center=<-3.1398 0 0>, rotation=<rot 90 0 0>}]
 dyn_undershirt_l_02 [{shapeType=Capsule, length=2, radius=3, shapeName=dyn_undershirt_01_capsule, center=<-3.1398 0 0>, rotation=<rot 90 0 0>}]
 dyn_undershirt_r_01 [{shapeType=Capsule, length=5, radius=2, shapeName=dyn_undershirt_01_capsule, center=<3.0918 0 0.0056>, rotation=<rot 0 0 90>}]
 dyn_undershirt_r_02 [{shapeType=Capsule, length=6, radius=3, shapeName=dyn_undershirt_01_capsule, center=<3.0917 0 0>, rotation=<rot 0 0 90>}]
 foot_l [{shapeType=Capsule, length=14.7126, radius=10, shapeName=foot_r_capsule}]
 foot_r [{shapeType=Capsule, length=14.7126, radius=10, shapeName=foot_r_capsule}]
 pelvis [{shapeType=Capsule, length=15, radius=9.5, rotation=<rot 90 0 0>}]
 spine_02 [{shapeType=Capsule, length=15, radius=13, shapeName=spine_02_capsule, center=<0 -3 0>, rotation=<rot 90 0 0>}]
 thigh_l [{shapeType=Capsule, length=20, radius=9.6, shapeName=thigh_l_capsule, center=<-10 -0.5 0>, rotation=<rot 90 0 0>}]
 thigh_r [
  {shapeType=Capsule, length=22, radius=9.5, shapeName=thigh_r_capsule, center=<10 0.6 0>, rotation=<rot 90 180 0>}
  {shapeType=Capsule, length=20, radius=9, shapeName=thigh_r_capsule, center=<10 0 -10>, rotation=<rot 90 180 0>}
 ]
 thigh_twist_01_l [{shapeType=Capsule, length=15, radius=10, shapeName=thigh_twist_01_l_capsule, center=<-7 0 0>, rotation=<rot 90 0 0>}]
 thigh_twist_01_r [{shapeType=Capsule, length=15, radius=10, shapeName=thigh_twist_01_r_capsule, center=<7 0 0>, rotation=<rot 90 180 0>}]
}
constraints [
 {bone1Name=dyn_undershirt_01, bone2Name=base_undershirt_01, swing1LimitDegrees=1, swing2LimitDegrees=70, twistLimitDegrees=5}
 {bone1Name=dyn_undershirt_02, bone2Name=dyn_undershirt_01, swing1LimitDegrees=1, swing2LimitDegrees=10, twistLimitDegrees=5}
 {bone1Name=dyn_undershirt_l_01, bone2Name=base_undershirt_l_01, swing1LimitDegrees=5, swing2LimitDegrees=55, twistLimitDegrees=1}
 {bone1Name=dyn_undershirt_l_02, bone2Name=dyn_undershirt_l_01, swing1LimitDegrees=5, swing2LimitDegrees=25, twistLimitDegrees=1}
 {bone1Name=dyn_undershirt_r_01, bone2Name=base_undershirt_r_01, swing1LimitDegrees=5, swing2LimitDegrees=70, twistLimitDegrees=1}
 {bone1Name=dyn_undershirt_r_02, bone2Name=dyn_undershirt_r_01, swing1LimitDegrees=5, swing2LimitDegrees=10, twistLimitDegrees=1}
 {
  bone1Name=thigh_l
  bone2Name=pelvis
  swing1LimitDegrees=45
  swing1Motion=Free
  swing2LimitDegrees=45
  swing2Motion=Free
  twistLimitDegrees=45
  twistMotion=Free
 }
 {
  bone1Name=thigh_r
  bone2Name=pelvis
  swing1LimitDegrees=45
  swing1Motion=Free
  swing2LimitDegrees=45
  swing2Motion=Free
  twistLimitDegrees=45
  twistMotion=Free
 }
 {bone1Name=dyn_undershirt_back_01, bone2Name=base_undershirt_back_01, swing1LimitDegrees=1, swing2LimitDegrees=30, twistLimitDegrees=1}
 {bone1Name=dyn_undershirt_back_02, bone2Name=dyn_undershirt_back_01, swing1LimitDegrees=1, swing2LimitDegrees=30, twistLimitDegrees=1}
 {bone1Name=dyn_undershirt_back_03, bone2Name=dyn_undershirt_back_02, swing1LimitDegrees=1, swing2LimitDegrees=30, twistLimitDegrees=1}
 {bone1Name=dyn_undershirt_back_l_01, bone2Name=base_undershirt_back_l_01, swing1LimitDegrees=1, swing2LimitDegrees=30, twistLimitDegrees=1}
 {bone1Name=dyn_undershirt_back_l_02, bone2Name=dyn_undershirt_back_l_01, swing1LimitDegrees=1, swing2LimitDegrees=30, twistLimitDegrees=1}
 {bone1Name=dyn_undershirt_back_l_03, bone2Name=dyn_undershirt_back_l_02, swing1LimitDegrees=1, swing2LimitDegrees=30, twistLimitDegrees=1}
 {bone1Name=dyn_undershirt_back_r_01, bone2Name=base_undershirt_back_r_01, swing1LimitDegrees=1, swing2LimitDegrees=1, twistLimitDegrees=1}
 {bone1Name=dyn_undershirt_back_r_02, bone2Name=dyn_undershirt_back_r_01, swing1LimitDegrees=1, swing2LimitDegrees=30, twistLimitDegrees=1}
 {bone1Name=dyn_undershirt_back_r_03, bone2Name=dyn_undershirt_back_r_02, swing1LimitDegrees=1, swing2LimitDegrees=30, twistLimitDegrees=1}
 {
  bone1Name=spine_02
  bone2Name=pelvis
  swing1LimitDegrees=45
  swing1Motion=Free
  swing2LimitDegrees=45
  swing2Motion=Free
  twistLimitDegrees=45
  twistMotion=Free
 }
 {bone1Name=dyn_outer_jacket_l_01, bone2Name=base_outer_jacket_l_01, swing1LimitDegrees=5, swing2LimitDegrees=20, twistLimitDegrees=1}
 {bone1Name=dyn_outer_jacket_l_03, bone2Name=base_outer_jacket_l_03, swing1LimitDegrees=5, swing2LimitDegrees=20, twistLimitDegrees=1}
 {bone1Name=dyn_outer_jacket_r_01, bone2Name=base_outer_jacket_r_01, swing1LimitDegrees=5, swing2LimitDegrees=30, twistLimitDegrees=1}
 {bone1Name=dyn_outer_jacket_r_03, bone2Name=base_outer_jacket_r_03, swing1LimitDegrees=5, swing2LimitDegrees=20, twistLimitDegrees=1}
 {bone1Name=dyn_outer_jacket_l_04, bone2Name=dyn_outer_jacket_l_03, swing1LimitDegrees=5, swing2LimitDegrees=20, twistLimitDegrees=1}
 {bone1Name=dyn_outer_jacket_r_04, bone2Name=dyn_outer_jacket_r_03, swing1LimitDegrees=5, swing2LimitDegrees=20, twistLimitDegrees=1}
 {
  bone1Name=base_outer_jacket_l_01
  bone2Name=spine_02
  swing1LimitDegrees=45
  swing1Motion=Free
  swing2LimitDegrees=45
  swing2Motion=Free
  twistLimitDegrees=45
  twistMotion=Locked
 }
 {
  bone1Name=base_outer_jacket_l_03
  bone2Name=spine_02
  swing1LimitDegrees=45
  swing1Motion=Free
  swing2LimitDegrees=45
  swing2Motion=Free
  twistLimitDegrees=45
  twistMotion=Locked
 }
 {
  bone1Name=base_outer_jacket_r_01
  bone2Name=spine_02
  swing1LimitDegrees=45
  swing1Motion=Free
  swing2LimitDegrees=45
  swing2Motion=Free
  twistLimitDegrees=45
  twistMotion=Free
 }
 {
  bone1Name=base_outer_jacket_r_03
  bone2Name=spine_02
  swing1LimitDegrees=45
  swing1Motion=Free
  swing2LimitDegrees=45
  swing2Motion=Free
  twistLimitDegrees=45
  twistMotion=Free
 }
 {
  bone1Name=base_undershirt_01
  bone2Name=pelvis
  swing1LimitDegrees=45
  swing1Motion=Free
  swing2LimitDegrees=45
  swing2Motion=Free
  twistLimitDegrees=45
  twistMotion=Free
 }
 {
  bone1Name=base_undershirt_l_01
  bone2Name=pelvis
  swing1LimitDegrees=45
  swing1Motion=Free
  swing2LimitDegrees=45
  swing2Motion=Free
  twistLimitDegrees=45
  twistMotion=Free
 }
 {
  bone1Name=base_undershirt_r_01
  bone2Name=pelvis
  swing1LimitDegrees=45
  swing1Motion=Free
  swing2LimitDegrees=45
  swing2Motion=Free
  twistLimitDegrees=45
  twistMotion=Free
 }
 {
  bone1Name=foot_l
  bone2Name=thigh_l
  swing1LimitDegrees=45
  swing1Motion=Free
  swing2LimitDegrees=45
  swing2Motion=Free
  twistLimitDegrees=45
  twistMotion=Free
 }
 {
  bone1Name=foot_r
  bone2Name=thigh_r
  swing1LimitDegrees=45
  swing1Motion=Free
  swing2LimitDegrees=45
  swing2Motion=Free
  twistLimitDegrees=45
  twistMotion=Free
 }
 {
  bone1Name=thigh_twist_01_l
  bone2Name=thigh_l
  swing1LimitDegrees=45
  swing1Motion=Free
  swing2LimitDegrees=45
  swing2Motion=Free
  twistLimitDegrees=45
  twistMotion=Free
 }
 {
  bone1Name=thigh_twist_01_r
  bone2Name=thigh_r
  swing1LimitDegrees=45
  swing1Motion=Free
  swing2LimitDegrees=45
  swing2Motion=Free
  twistLimitDegrees=45
  twistMotion=Free
 }
 {
  bone1Name=base_undershirt_back_01
  bone2Name=pelvis
  swing1LimitDegrees=45
  swing1Motion=Free
  swing2LimitDegrees=45
  swing2Motion=Free
  twistLimitDegrees=45
  twistMotion=Free
 }
 {
  bone1Name=base_undershirt_back_l_01
  bone2Name=pelvis
  swing1LimitDegrees=45
  swing1Motion=Free
  swing2LimitDegrees=45
  swing2Motion=Free
  twistLimitDegrees=45
  twistMotion=Free
 }
 {
  bone1Name=base_undershirt_back_r_01
  bone2Name=pelvis
  swing1LimitDegrees=45
  swing1Motion=Free
  swing2LimitDegrees=45
  swing2Motion=Free
  twistLimitDegrees=45
  twistMotion=Free
 }
 {bone1Name=dyn_minor_zip_jacket, bone2Name=dyn_outer_jacket_l_01, swing1LimitDegrees=1, swing2LimitDegrees=45, twistLimitDegrees=1}
]
