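# Physics asset for the M_MED_BentCarpet character's FaceAcc mesh, defining secondary-motion physics for hair and ponytail.
# Capsules on spine, clavicles, neck and head form the upper-body shapes, linked by free-rotation constraints.
# Hair strands hang from the head singly or in pairs, while a four-segment ponytail extends from a free-rotating base, with clumps branching off its first three segments.
# Tight swing and near-locked twist limits keep hair and ponytail close to their styled shape.
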PhysicsAsset /BRCosmetics/Characters/Player/Male/Medium/Bodies/M_MED_BentCarpet/Meshes/Parts/M_MED_BentCarpet_FaceAcc_Physics
bodies {
 base_ponytail_01 [{shapeType=Capsule, length=0.55, radius=0.55, shapeName=base_ponytail_01_capsule}]
 clavicle_l [{shapeType=Capsule, length=10, radius=5, shapeName=dyn_hood_01_capsule, center=<7 0 0>, rotation=<rot 90 0 0>}]
 clavicle_r [{shapeType=Capsule, length=10, radius=5, shapeName=dyn_hood_01_capsule, center=<-7 0 0>, rotation=<rot 90 0 0>}]
 dyn_hair_01 [{shapeType=Capsule, length=3, radius=1, shapeName=dyn_hair_01_capsule, center=<1.8 0 0>, rotation=<rot 90 0 0>}]
 dyn_hair_02 [{shapeType=Capsule, length=3, radius=1, shapeName=dyn_hair_01_capsule, center=<1.8 0 0>, rotation=<rot 90 0 0>}]
 dyn_hair_03 [{shapeType=Capsule, length=3, radius=1, shapeName=dyn_hair_01_capsule, center=<1.8 0 0>, rotation=<rot 90 0 0>}]
 dyn_hair_04 [{shapeType=Capsule, length=3, radius=1, shapeName=dyn_hair_01_capsule, center=<1.8 0 0>, rotation=<rot 90 0 0>}]
 dyn_hair_05 [{shapeType=Capsule, length=3, radius=1, shapeName=dyn_hair_01_capsule, center=<1.8 0 0>, rotation=<rot 90 0 0>}]
 dyn_hair_06 [{shapeType=Capsule, length=3, radius=1, shapeName=dyn_hair_01_capsule, center=<1.8 0 0>, rotation=<rot 90 0 0>}]
 dyn_hair_07 [{shapeType=Capsule, length=3, radius=1, shapeName=dyn_hair_01_capsule, center=<1.8 0 0>, rotation=<rot 90 0 0>}]
 dyn_hair_08 [{shapeType=Capsule, length=3, radius=1, shapeName=dyn_hair_01_capsule, center=<1.8 0 0>, rotation=<rot 90 0 0>}]
 dyn_hair_09 [{shapeType=Capsule, length=3, radius=1, shapeName=dyn_hair_01_capsule, center=<1.8 0 0>, rotation=<rot 90 0 0>}]
 dyn_hair_10 [{shapeType=Capsule, length=3, radius=1, shapeName=dyn_hair_01_capsule, center=<1.8 0 0>, rotation=<rot 90 0 0>}]
 dyn_hair_11 [{shapeType=Capsule, length=3, radius=1, shapeName=dyn_hair_01_capsule, center=<1.8 0 0>, rotation=<rot 90 0 0>}]
 dyn_hair_12 [{shapeType=Capsule, length=3, radius=1, shapeName=dyn_hair_01_capsule, center=<1.8 0 0>, rotation=<rot 90 0 0>}]
 dyn_hair_13 [{shapeType=Capsule, length=3, radius=1, shapeName=dyn_hair_01_capsule, center=<1.8 0 0>, rotation=<rot 90 0 0>}]
 dyn_ponytail_01 [{shapeType=Capsule, length=1, radius=5, shapeName=dyn_ponytail_01_capsule, center=<5.2615 0 0>, rotation=<rot 90 0 0>}]
 dyn_ponytail_02 [{shapeType=Capsule, length=0.5, radius=4.5, shapeName=dyn_ponytail_01_capsule, center=<3.2933 0 0>, rotation=<rot 90 0 0>}]
 dyn_ponytail_03 [{shapeType=Capsule, length=1, radius=3, shapeName=dyn_ponytail_01_capsule, center=<2.5722 0 0>, rotation=<rot 90 0 0>}]
 dyn_ponytail_04 [{shapeType=Capsule, length=2, radius=2, shapeName=dyn_ponytail_01_capsule, center=<2.0285 0 0>, rotation=<rot 90 0 0>}]
 dyn_ponytail_clump_01 [{shapeType=Capsule, length=3, radius=1, shapeName=dyn_ponytail_clump_01_capsule, center=<1.5743 0 0>, rotation=<rot 90 0 0>}]
 dyn_ponytail_clump_02 [{shapeType=Capsule, length=3, radius=1, shapeName=dyn_ponytail_clump_01_capsule, center=<1.5743 0 0>, rotation=<rot 90 0 0>}]
 dyn_ponytail_clump_03 [{shapeType=Capsule, length=3, radius=1, shapeName=dyn_ponytail_clump_01_capsule, center=<1.5743 0 0>, rotation=<rot 90 0 0>}]
 dyn_ponytail_clump_04 [{shapeType=Capsule, length=3, radius=1, shapeName=dyn_ponytail_clump_01_capsule, center=<1.5743 0 0>, rotation=<rot 90 0 0>}]
 dyn_ponytail_clump_05 [{shapeType=Capsule, length=3, radius=1, shapeName=dyn_ponytail_clump_01_capsule, center=<1.5743 0 0>, rotation=<rot 90 0 0>}]
 dyn_ponytail_clump_06 [{shapeType=Capsule, length=3, radius=1, shapeName=dyn_ponytail_clump_01_capsule, center=<1.5743 0 0>, rotation=<rot 90 0 0>}]
 dyn_ponytail_clump_07 [{shapeType=Capsule, length=3, radius=1, shapeName=dyn_ponytail_clump_01_capsule, center=<1.5743 0 0>, rotation=<rot 90 0 0>}]
 dyn_ponytail_clump_08 [{shapeType=Capsule, length=3, radius=1, shapeName=dyn_ponytail_clump_01_capsule, center=<1.5743 0 0>, rotation=<rot 90 0 0>}]
 dyn_ponytail_clump_09 [{shapeType=Capsule, length=3, radius=1, shapeName=dyn_ponytail_clump_01_capsule, center=<1.5743 0 0>, rotation=<rot 90 0 0>}]
 head [{shapeType=Capsule, length=1, radius=6, shapeName=head_capsule, center=<10 2 0>}]
 neck_01 [{shapeType=Capsule, length=20, radius=8, shapeName=dyn_hood_01_capsule, center=<-4.6543 3.2738 0>}]
 spine_05 [{shapeType=Capsule, length=20, radius=8, shapeName=dyn_hood_01_capsule, center=<9.0981 -6.0118 0>}]
}
constraints [
 {
  bone1Name=base_ponytail_01
  bone2Name=head
  swing1LimitDegrees=45
  swing1Motion=Free
  swing2LimitDegrees=45
  swing2Motion=Free
  twistLimitDegrees=45
  twistMotion=Free
 }
 {bone1Name=dyn_ponytail_01, bone2Name=base_ponytail_01, swing1LimitDegrees=15, swing2LimitDegrees=15, twistLimitDegrees=1}
 {bone1Name=dyn_ponytail_02, bone2Name=dyn_ponytail_01, swing1LimitDegrees=15, swing2LimitDegrees=15, twistLimitDegrees=1}
 {bone1Name=dyn_ponytail_03, bone2Name=dyn_ponytail_02, swing1LimitDegrees=15, swing2LimitDegrees=15, twistLimitDegrees=1}
 {bone1Name=dyn_ponytail_04, bone2Name=dyn_ponytail_03, swing1LimitDegrees=15, swing2LimitDegrees=15, twistLimitDegrees=1}
 {
  bone1Name=head
  bone2Name=neck_01
  swing1LimitDegrees=45
  swing1Motion=Free
  swing2LimitDegrees=45
  swing2Motion=Free
  twistLimitDegrees=45
  twistMotion=Free
 }
 {
  bone1Name=clavicle_l
  bone2Name=spine_05
  swing1LimitDegrees=45
  swing1Motion=Free
  swing2LimitDegrees=45
  swing2Motion=Free
  twistLimitDegrees=45
  twistMotion=Free
 }
 {
  bone1Name=clavicle_r
  bone2Name=spine_05
  swing1LimitDegrees=45
  swing1Motion=Free
  swing2LimitDegrees=45
  swing2Motion=Free
  twistLimitDegrees=45
  twistMotion=Free
 }
 {bone1Name=dyn_hair_01, bone2Name=head, swing1LimitDegrees=5, swing2LimitDegrees=20, twistLimitDegrees=1}
 {bone1Name=dyn_hair_02, bone2Name=dyn_hair_01, swing1LimitDegrees=5, swing2LimitDegrees=20, twistLimitDegrees=1}
 {bone1Name=dyn_hair_03, bone2Name=head, swing1LimitDegrees=5, swing2LimitDegrees=20, twistLimitDegrees=1}
 {bone1Name=dyn_hair_04, bone2Name=dyn_hair_03, swing1LimitDegrees=5, swing2LimitDegrees=20, twistLimitDegrees=1}
 {bone1Name=dyn_hair_05, bone2Name=head, swing1LimitDegrees=5, swing2LimitDegrees=20, twistLimitDegrees=1}
 {bone1Name=dyn_hair_06, bone2Name=dyn_hair_05, swing1LimitDegrees=5, swing2LimitDegrees=20, twistLimitDegrees=1}
 {bone1Name=dyn_hair_07, bone2Name=head, swing1LimitDegrees=5, swing2LimitDegrees=5, twistLimitDegrees=1}
 {bone1Name=dyn_hair_08, bone2Name=dyn_hair_07, swing1LimitDegrees=5, swing2LimitDegrees=5, twistLimitDegrees=1}
 {bone1Name=dyn_hair_09, bone2Name=head, swing1LimitDegrees=5, swing2LimitDegrees=20, twistLimitDegrees=1}
 {bone1Name=dyn_hair_10, bone2Name=head, swing1LimitDegrees=5, swing2LimitDegrees=20, twistLimitDegrees=1}
 {bone1Name=dyn_hair_11, bone2Name=head, swing1LimitDegrees=5, swing2LimitDegrees=20, twistLimitDegrees=1}
 {bone1Name=dyn_hair_12, bone2Name=head, swing1LimitDegrees=5, swing2LimitDegrees=20, twistLimitDegrees=1}
 {bone1Name=dyn_hair_13, bone2Name=head, swing1LimitDegrees=5, swing2LimitDegrees=20, twistLimitDegrees=1}
 {bone1Name=dyn_ponytail_clump_01, bone2Name=dyn_ponytail_01, swing1LimitDegrees=5, swing2LimitDegrees=5, twistLimitDegrees=1}
 {bone1Name=dyn_ponytail_clump_02, bone2Name=dyn_ponytail_clump_01, swing1LimitDegrees=5, swing2LimitDegrees=5, twistLimitDegrees=1}
 {bone1Name=dyn_ponytail_clump_03, bone2Name=dyn_ponytail_02, swing1LimitDegrees=5, swing2LimitDegrees=5, twistLimitDegrees=1}
 {bone1Name=dyn_ponytail_clump_04, bone2Name=dyn_ponytail_clump_03, swing1LimitDegrees=5, swing2LimitDegrees=5, twistLimitDegrees=1}
 {bone1Name=dyn_ponytail_clump_05, bone2Name=dyn_ponytail_02, swing1LimitDegrees=5, swing2LimitDegrees=5, twistLimitDegrees=1}
 {bone1Name=dyn_ponytail_clump_06, bone2Name=dyn_ponytail_02, swing1LimitDegrees=5, swing2LimitDegrees=5, twistLimitDegrees=1}
 {bone1Name=dyn_ponytail_clump_07, bone2Name=dyn_ponytail_clump_06, swing1LimitDegrees=5, swing2LimitDegrees=5, twistLimitDegrees=1}
 {bone1Name=dyn_ponytail_clump_08, bone2Name=dyn_ponytail_03, swing1LimitDegrees=5, swing2LimitDegrees=5, twistLimitDegrees=1}
 {bone1Name=dyn_ponytail_clump_09, bone2Name=dyn_ponytail_03, swing1LimitDegrees=5, swing2LimitDegrees=5, twistLimitDegrees=1}
 {
  bone1Name=neck_01
  bone2Name=spine_05
  swing1LimitDegrees=45
  swing1Motion=Free
  swing2LimitDegrees=45
  swing2Motion=Free
  twistLimitDegrees=45
  twistMotion=Free
 }
]
